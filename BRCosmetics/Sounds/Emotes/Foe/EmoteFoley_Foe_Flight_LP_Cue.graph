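# Looping foley sound cue for the Foe flight emote, layering a continuous flight loop with occasional laughs.
# A mixer combines the looping flight wave, slightly boosted, with a quieter looping branch that waits a random delay, then plays one of three equally weighted laugh variants.
# It uses emote attenuation, concurrency and foley class settings, keeps playing when silent, and sends to a music audio-analysis submix.
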SoundCue /BRCosmetics/Sounds/Emotes/Foe/EmoteFoley_Foe_Flight_LP_Cue
root=Mixer_0
nodes [
 {id=Delay_0, children [Random_0], properties {delayMax=8, delayMin=1}}
 {id=Looping_0, children [Delay_0]}
 {id=Mixer_0, children [WavePlayer_0,Looping_0], properties {inputVolume [1.3,0.8]}}
 {id=Random_0, children [WavePlayer_2,WavePlayer_3,WavePlayer_4], properties {weights [1,1,1]}}
 {id=WavePlayer_0, properties {bLooping=true, soundWaveAssetPtr=&/Emote_Foe_Flight_LP}}
 {id=WavePlayer_2, properties {soundWaveAssetPtr=&/Emote_Foe_Laugh_01}}
 {id=WavePlayer_3, properties {soundWaveAssetPtr=&/Emote_Foe_Laugh_02}}
 {id=WavePlayer_4, properties {soundWaveAssetPtr=&/Emote_Foe_Laugh_03}}
]
properties {
 attenuationSettings=&Fort_Emotes_Athena_Att
 audioPropertiesSheet=/BRCosmetics/Sounds/Settings/PropertySheets/Emotes/APS_Cosmetic_Emote_Foley
 maxDistance=1600
 soundClassObject=/Game/Sounds/FortSoundClassesAndModes/Fort_Foley_Emote
 virtualizationMode=PlayWhenSilent
 volumeMultiplier=1
 concurrencyOverrides {maxCount=3}
 concurrencySet [/Game/Sounds/Concurrency_Rules/Athena_EmoteFoley_Conc]
 soundSubmixSends [{minSendLevel=0, soundSubmix=&Fortnite_AudioAnalysis_Music}]
}
